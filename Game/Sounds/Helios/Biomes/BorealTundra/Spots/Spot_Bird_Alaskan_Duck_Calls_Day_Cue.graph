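# Sound cue for daytime Alaskan duck calls, a one-shot bird spot ambience in the Boreal Tundra biome.
# A quality-level switch feeds two Modulator branches, each shared by two levels and randomly picking from the same five equally weighted call waves.
# It plays quietly in the outdoor one-shot sound class with bird tile-ambient attenuation and sends to the exterior ambient audio bus.
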SoundCue /Game/Sounds/Helios/Biomes/BorealTundra/Spots/Spot_Bird_Alaskan_Duck_Calls_Day_Cue
root=QualityLevel_0
nodes [
 {id=Modulator_0, children [Random_0]}
 {id=Modulator_3, children [Random_1]}
 {id=QualityLevel_0, children [Modulator_0,Modulator_0,Modulator_3,Modulator_3]}
 {id=Random_0, children [WavePlayer_8, WavePlayer_9, WavePlayer_10, WavePlayer_11, WavePlayer_12], properties {weights [1,1,1,1,1]}}
 {id=Random_1, children [WavePlayer_13, WavePlayer_14, WavePlayer_15, WavePlayer_16, WavePlayer_17], properties {weights [1,1,1,1,1]}}
 {id=WavePlayer_10, properties {soundWaveAssetPtr=&/Spot_Bird_Alaskan_Duck_Calls_Day_03}}
 {id=WavePlayer_11, properties {soundWaveAssetPtr=&/Spot_Bird_Alaskan_Duck_Calls_Day_04}}
 {id=WavePlayer_12, properties {soundWaveAssetPtr=&/Spot_Bird_Alaskan_Duck_Calls_Day_05}}
 {id=WavePlayer_13, properties {soundWaveAssetPtr=&/Spot_Bird_Alaskan_Duck_Calls_Day_01}}
 {id=WavePlayer_14, properties {soundWaveAssetPtr=&/Spot_Bird_Alaskan_Duck_Calls_Day_02}}
 {id=WavePlayer_15, properties {soundWaveAssetPtr=&Spot_Bird_Alaskan_Duck_Calls_Day_03}}
 {id=WavePlayer_16, properties {soundWaveAssetPtr=&Spot_Bird_Alaskan_Duck_Calls_Day_04}}
 {id=WavePlayer_17, properties {soundWaveAssetPtr=&Spot_Bird_Alaskan_Duck_Calls_Day_05}}
 {id=WavePlayer_8, properties {soundWaveAssetPtr=&Spot_Bird_Alaskan_Duck_Calls_Day_01}}
 {id=WavePlayer_9, properties {soundWaveAssetPtr=&Spot_Bird_Alaskan_Duck_Calls_Day_02}}
]
properties {
 attenuationSettings=&Fort_TileAmbient_OneShot_Bird_Att
 maxDistance=8005
 soundClassObject=&Fort_Ambient_OneShots_Outdoors
 volumeMultiplier=0.2
 busSends [{audioBus=null, minSendLevel=0.6, sendLevel=1, soundSourceBus=&AmbientAudio_Exterior_SourceBus}]
}
