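# Sound cue for arctic snow fowl calls, an outdoor one-shot bird ambience spot for the Boreal Tundra biome.
# A quality-level root feeds its first two slots and last two slots into separate Modulator and Random branches.
# Each branch picks evenly among four identically named evening call waves; the second is excluded from branch culling.
# The quiet cue uses tile-ambient bird attenuation and sends to the exterior ambient source bus.
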SoundCue /Game/Sounds/Helios/Biomes/BorealTundra/Spots/Spot_Bird_Arctic_Snow_Fowl_Calls_Cue
root=QualityLevel_0
nodes [
 {id=Modulator_0, children [Random_2]}
 {id=Modulator_3, children [Random_3]}
 {id=QualityLevel_0, children [Modulator_0,Modulator_0,Modulator_3,Modulator_3]}
 {id=Random_2, children [WavePlayer_0,WavePlayer_1,WavePlayer_2,WavePlayer_3], properties {weights [1,1,1,1]}}
 {
  id=Random_3
  children [WavePlayer_4,WavePlayer_5,WavePlayer_6,WavePlayer_7]
  properties {bShouldExcludeFromBranchCulling=true, weights [1,1,1,1]}
 }
 {id=WavePlayer_0, properties {soundWaveAssetPtr=&/Spot_Bird_Arctic_Snow_Fowl_Calls_Evening_01}}
 {id=WavePlayer_1, properties {soundWaveAssetPtr=&/Spot_Bird_Arctic_Snow_Fowl_Calls_Evening_02}}
 {id=WavePlayer_2, properties {soundWaveAssetPtr=&/Spot_Bird_Arctic_Snow_Fowl_Calls_Evening_03}}
 {id=WavePlayer_3, properties {soundWaveAssetPtr=&/Spot_Bird_Arctic_Snow_Fowl_Calls_Evening_04}}
 {id=WavePlayer_4, properties {soundWaveAssetPtr=&Spot_Bird_Arctic_Snow_Fowl_Calls_Evening_01}}
 {id=WavePlayer_5, properties {soundWaveAssetPtr=&Spot_Bird_Arctic_Snow_Fowl_Calls_Evening_02}}
 {id=WavePlayer_6, properties {soundWaveAssetPtr=&Spot_Bird_Arctic_Snow_Fowl_Calls_Evening_03}}
 {id=WavePlayer_7, properties {soundWaveAssetPtr=&Spot_Bird_Arctic_Snow_Fowl_Calls_Evening_04}}
]
properties {
 attenuationSettings=&Fort_TileAmbient_OneShot_Bird_Att
 maxDistance=8005
 soundClassObject=&Fort_Ambient_OneShots_Outdoors
 volumeMultiplier=0.2
 busSends [{
   audioBus=null
   maxSendDistance=6000
   minSendDistance=3000
   minSendLevel=0.6
   sendLevel=1
   soundSourceBus=&AmbientAudio_Exterior_SourceBus
  }]
}
