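# Sound cue for bar-tailed godwit chirps, an outdoor ambient one-shot spot sound in the Boreal Tundra biome.
# A quality-level root sends its first two levels to a random pick of ten waves and its last two to a five-wave subset (06–10).
# Each random pick feeds a modulator for pitch and volume variation.
# The cue uses bird one-shot attenuation, reduced volume and a distance-scaled send to the exterior ambient bus.
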SoundCue /Game/Sounds/Helios/Biomes/BorealTundra/Spots/Spot_Bird_BarTailedGodwit_Chirps_Cue
root=QualityLevel_0
nodes [
 {id=Modulator_0, children [Random_0]}
 {id=Modulator_3, children [Random_1]}
 {id=QualityLevel_0, children [Modulator_0,Modulator_0,Modulator_3,Modulator_3]}
 {
  id=Random_0
  children [
   WavePlayer_18
   WavePlayer_19
   WavePlayer_20
   WavePlayer_21
   WavePlayer_22
   WavePlayer_23
   WavePlayer_24
   WavePlayer_25
   WavePlayer_26
   WavePlayer_27
  ]
  properties {weights [1,1,1,1,1,1,1,1,1,1]}
 }
 {id=Random_1, children [WavePlayer_28, WavePlayer_29, WavePlayer_30, WavePlayer_31, WavePlayer_32], properties {weights [1,1,1,1,1]}}
 {id=WavePlayer_18, properties {soundWaveAssetPtr=&/Spot_Bird_BarTailedGodwit_Chirps_01}}
 {id=WavePlayer_19, properties {soundWaveAssetPtr=&/Spot_Bird_BarTailedGodwit_Chirps_02}}
 {id=WavePlayer_20, properties {soundWaveAssetPtr=&/Spot_Bird_BarTailedGodwit_Chirps_03}}
 {id=WavePlayer_21, properties {soundWaveAssetPtr=&/Spot_Bird_BarTailedGodwit_Chirps_04}}
 {id=WavePlayer_22, properties {soundWaveAssetPtr=&/Spot_Bird_BarTailedGodwit_Chirps_05}}
 {id=WavePlayer_23, properties {soundWaveAssetPtr=&/Spot_Bird_BarTailedGodwit_Chirps_06}}
 {id=WavePlayer_24, properties {soundWaveAssetPtr=&/Spot_Bird_BarTailedGodwit_Chirps_07}}
 {id=WavePlayer_25, properties {soundWaveAssetPtr=&/Spot_Bird_BarTailedGodwit_Chirps_08}}
 {id=WavePlayer_26, properties {soundWaveAssetPtr=&/Spot_Bird_BarTailedGodwit_Chirps_09}}
 {id=WavePlayer_27, properties {soundWaveAssetPtr=&/Spot_Bird_BarTailedGodwit_Chirps_10}}
 {id=WavePlayer_28, properties {soundWaveAssetPtr=&Spot_Bird_BarTailedGodwit_Chirps_08}}
 {id=WavePlayer_29, properties {soundWaveAssetPtr=&Spot_Bird_BarTailedGodwit_Chirps_06}}
 {id=WavePlayer_30, properties {soundWaveAssetPtr=&Spot_Bird_BarTailedGodwit_Chirps_09}}
 {id=WavePlayer_31, properties {soundWaveAssetPtr=&Spot_Bird_BarTailedGodwit_Chirps_10}}
 {id=WavePlayer_32, properties {soundWaveAssetPtr=&Spot_Bird_BarTailedGodwit_Chirps_07}}
]
properties {
 attenuationSettings=&Fort_TileAmbient_OneShot_Bird_Att
 maxDistance=8005
 soundClassObject=&Fort_Ambient_OneShots_Outdoors
 volumeMultiplier=0.251
 busSends [{
   audioBus=null
   maxSendDistance=6000
   minSendDistance=3000
   minSendLevel=0.6
   sendLevel=1
   soundSourceBus=&AmbientAudio_Exterior_SourceBus
  }]
}
