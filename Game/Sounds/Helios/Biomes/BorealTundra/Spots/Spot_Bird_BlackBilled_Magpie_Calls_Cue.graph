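# Sound cue for black-billed magpie call one-shots, an ambient spot sound in the Helios BorealTundra biome.
# A quality-level root shares two branches across its four slots, two slots each.
# Each branch runs a modulator into an equally weighted random pick of four call waves, with matching names but different asset references per branch.
# Output is quiet, uses bird one-shot attenuation and the outdoor ambient class, and sends to the exterior ambience source bus.
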SoundCue /Game/Sounds/Helios/Biomes/BorealTundra/Spots/Spot_Bird_BlackBilled_Magpie_Calls_Cue
root=QualityLevel_0
nodes [
 {id=Modulator_0, children [Random_2]}
 {id=Modulator_3, children [Random_3]}
 {id=QualityLevel_0, children [Modulator_0,Modulator_0,Modulator_3,Modulator_3]}
 {id=Random_2, children [WavePlayer_6,WavePlayer_7,WavePlayer_8,WavePlayer_9], properties {weights [1,1,1,1]}}
 {id=Random_3, children [WavePlayer_10, WavePlayer_11, WavePlayer_12, WavePlayer_13], properties {weights [1,1,1,1]}}
 {id=WavePlayer_10, properties {soundWaveAssetPtr=&/Spot_Bird_BlackBilled_Magpie_Calls_01}}
 {id=WavePlayer_11, properties {soundWaveAssetPtr=&/Spot_Bird_BlackBilled_Magpie_Calls_02}}
 {id=WavePlayer_12, properties {soundWaveAssetPtr=&/Spot_Bird_BlackBilled_Magpie_Calls_03}}
 {id=WavePlayer_13, properties {soundWaveAssetPtr=&/Spot_Bird_BlackBilled_Magpie_Calls_04}}
 {id=WavePlayer_6, properties {soundWaveAssetPtr=&Spot_Bird_BlackBilled_Magpie_Calls_01}}
 {id=WavePlayer_7, properties {soundWaveAssetPtr=&Spot_Bird_BlackBilled_Magpie_Calls_02}}
 {id=WavePlayer_8, properties {soundWaveAssetPtr=&Spot_Bird_BlackBilled_Magpie_Calls_03}}
 {id=WavePlayer_9, properties {soundWaveAssetPtr=&Spot_Bird_BlackBilled_Magpie_Calls_04}}
]
properties {
 attenuationSettings=&Fort_TileAmbient_OneShot_Bird_Att
 maxDistance=8005
 soundClassObject=&Fort_Ambient_OneShots_Outdoors
 volumeMultiplier=0.15
 busSends [{
   audioBus=null
   maxSendDistance=6000
   minSendDistance=3000
   minSendLevel=0.6
   sendLevel=1
   soundSourceBus=&AmbientAudio_Exterior_SourceBus
  }]
}
