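# Sound cue for common mynah morning calls, an outdoor one-shot ambient bird spot in the Boreal Tundra biome.
# A quality-level node sends the first two levels to a random pick of eight call variations and the last two to a smaller pool of four from the same set.
# Each pool feeds through its own modulator, with equal weights across variations.
# Playback uses shared bird one-shot attenuation, the outdoor ambient sound class and a reduced volume.
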SoundCue /Game/Sounds/Helios/Biomes/BorealTundra/Spots/Spot_Bird_Common_Mynah_Morning_Calls_Cue
root=QualityLevel_0
nodes [
 {id=Modulator_0, children [Random_2]}
 {id=Modulator_3, children [Random_3]}
 {id=QualityLevel_0, children [Modulator_0,Modulator_0,Modulator_3,Modulator_3]}
 {
  id=Random_2
  children [WavePlayer_6, WavePlayer_7, WavePlayer_8, WavePlayer_9, WavePlayer_10, WavePlayer_11, WavePlayer_12, WavePlayer_13]
  properties {weights [1,1,1,1,1,1,1,1]}
 }
 {id=Random_3, children [WavePlayer_14, WavePlayer_15, WavePlayer_19, WavePlayer_20], properties {weights [1,1,1,1]}}
 {id=WavePlayer_10, properties {soundWaveAssetPtr=&/Spot_Bird_Common_Mynah_Morning_Calls_05}}
 {id=WavePlayer_11, properties {soundWaveAssetPtr=&/Spot_Bird_Common_Mynah_Morning_Calls_06}}
 {id=WavePlayer_12, properties {soundWaveAssetPtr=&/Spot_Bird_Common_Mynah_Morning_Calls_07}}
 {id=WavePlayer_13, properties {soundWaveAssetPtr=&/Spot_Bird_Common_Mynah_Morning_Calls_08}}
 {id=WavePlayer_14, properties {soundWaveAssetPtr=&Spot_Bird_Common_Mynah_Morning_Calls_05}}
 {id=WavePlayer_15, properties {soundWaveAssetPtr=&Spot_Bird_Common_Mynah_Morning_Calls_06}}
 {id=WavePlayer_19, properties {soundWaveAssetPtr=&/Spot_Bird_Common_Mynah_Morning_Calls_04}}
 {id=WavePlayer_20, properties {soundWaveAssetPtr=&/Spot_Bird_Common_Mynah_Morning_Calls_03}}
 {id=WavePlayer_6, properties {soundWaveAssetPtr=&/Spot_Bird_Common_Mynah_Morning_Calls_01}}
 {id=WavePlayer_7, properties {soundWaveAssetPtr=&/Spot_Bird_Common_Mynah_Morning_Calls_02}}
 {id=WavePlayer_8, properties {soundWaveAssetPtr=&Spot_Bird_Common_Mynah_Morning_Calls_03}}
 {id=WavePlayer_9, properties {soundWaveAssetPtr=&Spot_Bird_Common_Mynah_Morning_Calls_04}}
]
properties {
 attenuationSettings=&Fort_TileAmbient_OneShot_Bird_Att
 maxDistance=8005
 soundClassObject=&Fort_Ambient_OneShots_Outdoors
 volumeMultiplier=0.251
 busSends [{audioBus=null, minSendLevel=0, sendLevel=1, soundSourceBus=null}]
}
