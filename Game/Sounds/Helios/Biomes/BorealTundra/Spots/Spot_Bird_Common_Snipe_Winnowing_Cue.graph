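# Sound cue for a common snipe winnowing call, an outdoor ambient bird one-shot in the boreal tundra biome.
# A quality-level node routes the first two levels to a five-wave random pool and the other two to a three-wave subset, each through a Modulator.
# It plays at low volume with bird one-shot attenuation and a distance-based send to the exterior ambient source bus.
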SoundCue /Game/Sounds/Helios/Biomes/BorealTundra/Spots/Spot_Bird_Common_Snipe_Winnowing_Cue
root=QualityLevel_0
nodes [
 {id=Modulator_0, children [Random_0]}
 {id=Modulator_3, children [Random_1]}
 {id=QualityLevel_0, children [Modulator_0,Modulator_0,Modulator_3,Modulator_3]}
 {id=Random_0, children [WavePlayer_0,WavePlayer_1,WavePlayer_2,WavePlayer_4,WavePlayer_5], properties {weights [1,1,1,1,1]}}
 {id=Random_1, children [WavePlayer_16, WavePlayer_17, WavePlayer_18], properties {weights [1,1,1]}}
 {id=WavePlayer_0, properties {soundWaveAssetPtr=&/Spot_Bird_Common_Snipe_Winnowing_01}}
 {id=WavePlayer_1, properties {soundWaveAssetPtr=&/Spot_Bird_Common_Snipe_Winnowing_02}}
 {id=WavePlayer_16, properties {soundWaveAssetPtr=&Spot_Bird_Common_Snipe_Winnowing_01}}
 {id=WavePlayer_17, properties {soundWaveAssetPtr=&Spot_Bird_Common_Snipe_Winnowing_02}}
 {id=WavePlayer_18, properties {soundWaveAssetPtr=&/Spot_Bird_Common_Snipe_Winnowing_05}}
 {id=WavePlayer_2, properties {soundWaveAssetPtr=&/Spot_Bird_Common_Snipe_Winnowing_03}}
 {id=WavePlayer_4, properties {soundWaveAssetPtr=&Spot_Bird_Common_Snipe_Winnowing_05}}
 {id=WavePlayer_5, properties {soundWaveAssetPtr=&/Spot_Bird_Common_Snipe_Winnowing_06}}
]
properties {
 attenuationSettings=&Fort_TileAmbient_OneShot_Bird_Att
 maxDistance=8005
 soundClassObject=&Fort_Ambient_OneShots_Outdoors
 volumeMultiplier=0.15
 busSends [{
   audioBus=null
   maxSendDistance=6000
   minSendDistance=3000
   minSendLevel=0.6
   sendLevel=1
   soundSourceBus=&AmbientAudio_Exterior_SourceBus
  }]
}
