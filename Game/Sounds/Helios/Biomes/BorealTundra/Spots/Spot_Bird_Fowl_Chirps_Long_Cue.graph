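# Boreal tundra ambient spot cue that plays long fowl chirp one-shots, using night-named recordings, as outdoor bird ambience.
# A quality-level switch feeds two branches, each shared by two levels, where a modulator varies a random pick from five equally weighted waves.
# The second branch's random node is excluded from branch culling.
# Bird one-shot attenuation applies, with a distance-scaled send to the exterior ambient source bus.
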SoundCue /Game/Sounds/Helios/Biomes/BorealTundra/Spots/Spot_Bird_Fowl_Chirps_Long_Cue
root=QualityLevel_0
nodes [
 {id=Modulator_3, children [Random_4]}
 {id=Modulator_4, children [Random_7]}
 {id=QualityLevel_0, children [Modulator_3,Modulator_3,Modulator_4,Modulator_4]}
 {id=Random_4, children [WavePlayer_32, WavePlayer_33, WavePlayer_34, WavePlayer_35, WavePlayer_36], properties {weights [1,1,1,1,1]}}
 {
  id=Random_7
  children [WavePlayer_37, WavePlayer_38, WavePlayer_39, WavePlayer_40, WavePlayer_41]
  properties {bShouldExcludeFromBranchCulling=true, weights [1,1,1,1,1]}
 }
 {id=WavePlayer_32, properties {soundWaveAssetPtr=&/Spot_Bird_Fowl_Chirps_Long_Night_01}}
 {id=WavePlayer_33, properties {soundWaveAssetPtr=&/Spot_Bird_Fowl_Chirps_Long_Night_02}}
 {id=WavePlayer_34, properties {soundWaveAssetPtr=&/Spot_Bird_Fowl_Chirps_Long_Night_03}}
 {id=WavePlayer_35, properties {soundWaveAssetPtr=&/Spot_Bird_Fowl_Chirps_Long_Night_04}}
 {id=WavePlayer_36, properties {soundWaveAssetPtr=&/Spot_Bird_Fowl_Chirps_Long_Night_05}}
 {id=WavePlayer_37, properties {soundWaveAssetPtr=&Spot_Bird_Fowl_Chirps_Long_Night_01}}
 {id=WavePlayer_38, properties {soundWaveAssetPtr=&Spot_Bird_Fowl_Chirps_Long_Night_02}}
 {id=WavePlayer_39, properties {soundWaveAssetPtr=&Spot_Bird_Fowl_Chirps_Long_Night_03}}
 {id=WavePlayer_40, properties {soundWaveAssetPtr=&Spot_Bird_Fowl_Chirps_Long_Night_04}}
 {id=WavePlayer_41, properties {soundWaveAssetPtr=&Spot_Bird_Fowl_Chirps_Long_Night_05}}
]
properties {
 attenuationSettings=&Fort_TileAmbient_OneShot_Bird_Att
 maxDistance=8005
 soundClassObject=&Fort_Ambient_OneShots_Outdoors
 volumeMultiplier=0.251
 busSends [{
   audioBus=null
   maxSendDistance=6000
   minSendDistance=3000
   minSendLevel=0.6
   sendLevel=1
   soundSourceBus=&AmbientAudio_Exterior_SourceBus
  }]
}
