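# Sound cue for one-shot Himalayan snow tern morning calls in the Helios Boreal Tundra biome, using outdoor ambient sound class and bird attenuation.
# A quality-level root sends the first two levels to a fifteen-wave random pool and the other two to a four-wave pool (Day_12 to 15).
# Each equally weighted pool feeds its own modulator for pitch and volume variation.
# The smaller pool is excluded from branch culling.
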SoundCue /Game/Sounds/Helios/Biomes/BorealTundra/Spots/Spot_Bird_Himalayan_Snow_Tern_Morning_Cue
root=QualityLevel_0
nodes [
 {id=Modulator_0, children [Random_3]}
 {id=Modulator_3, children [Random_2]}
 {id=QualityLevel_0, children [Modulator_3,Modulator_3,Modulator_0,Modulator_0]}
 {
  id=Random_2
  children [
   WavePlayer_8
   WavePlayer_9
   WavePlayer_10
   WavePlayer_11
   WavePlayer_12
   WavePlayer_13
   WavePlayer_14
   WavePlayer_15
   WavePlayer_16
   WavePlayer_17
   WavePlayer_18
   WavePlayer_24
   WavePlayer_25
   WavePlayer_26
   WavePlayer_27
  ]
  properties {weights [1,1,1,1,1,1,1,1,1,1,1,1,1,1,1]}
 }
 {
  id=Random_3
  children [WavePlayer_28, WavePlayer_29, WavePlayer_30, WavePlayer_31]
  properties {bShouldExcludeFromBranchCulling=true, weights [1,1,1,1]}
 }
 {id=WavePlayer_10, properties {soundWaveAssetPtr=&/Spot_Bird_Himalayan_Snow_Tern_Morning_Day_03}}
 {id=WavePlayer_11, properties {soundWaveAssetPtr=&/Spot_Bird_Himalayan_Snow_Tern_Morning_Day_04}}
 {id=WavePlayer_12, properties {soundWaveAssetPtr=&/Spot_Bird_Himalayan_Snow_Tern_Morning_Day_05}}
 {id=WavePlayer_13, properties {soundWaveAssetPtr=&/Spot_Bird_Himalayan_Snow_Tern_Morning_Day_06}}
 {id=WavePlayer_14, properties {soundWaveAssetPtr=&/Spot_Bird_Himalayan_Snow_Tern_Morning_Day_07}}
 {id=WavePlayer_15, properties {soundWaveAssetPtr=&/Spot_Bird_Himalayan_Snow_Tern_Morning_Day_08}}
 {id=WavePlayer_16, properties {soundWaveAssetPtr=&/Spot_Bird_Himalayan_Snow_Tern_Morning_Day_09}}
 {id=WavePlayer_17, properties {soundWaveAssetPtr=&/Spot_Bird_Himalayan_Snow_Tern_Morning_Day_10}}
 {id=WavePlayer_18, properties {soundWaveAssetPtr=&/Spot_Bird_Himalayan_Snow_Tern_Morning_Day_11}}
 {id=WavePlayer_24, properties {soundWaveAssetPtr=&/Spot_Bird_Himalayan_Snow_Tern_Morning_Day_12}}
 {id=WavePlayer_25, properties {soundWaveAssetPtr=&/Spot_Bird_Himalayan_Snow_Tern_Morning_Day_13}}
 {id=WavePlayer_26, properties {soundWaveAssetPtr=&/Spot_Bird_Himalayan_Snow_Tern_Morning_Day_14}}
 {id=WavePlayer_27, properties {soundWaveAssetPtr=&/Spot_Bird_Himalayan_Snow_Tern_Morning_Day_15}}
 {id=WavePlayer_28, properties {soundWaveAssetPtr=&Spot_Bird_Himalayan_Snow_Tern_Morning_Day_12}}
 {id=WavePlayer_29, properties {soundWaveAssetPtr=&Spot_Bird_Himalayan_Snow_Tern_Morning_Day_13}}
 {id=WavePlayer_30, properties {soundWaveAssetPtr=&Spot_Bird_Himalayan_Snow_Tern_Morning_Day_14}}
 {id=WavePlayer_31, properties {soundWaveAssetPtr=&Spot_Bird_Himalayan_Snow_Tern_Morning_Day_15}}
 {id=WavePlayer_8, properties {soundWaveAssetPtr=&/Spot_Bird_Himalayan_Snow_Tern_Morning_Day_01}}
 {id=WavePlayer_9, properties {soundWaveAssetPtr=&/Spot_Bird_Himalayan_Snow_Tern_Morning_Day_02}}
]
properties {
 attenuationSettings=&Fort_TileAmbient_OneShot_Bird_Att
 maxDistance=8005
 soundClassObject=&Fort_Ambient_OneShots_Outdoors
 volumeMultiplier=0.401
 busSends [{audioBus=null, minSendLevel=0, sendLevel=1, soundSourceBus=null}]
}
