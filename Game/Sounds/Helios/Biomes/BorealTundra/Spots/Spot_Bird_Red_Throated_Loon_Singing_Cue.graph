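# Boreal tundra ambient one-shot cue that plays a quiet red-throated loon song as an outdoor spot sound.
# A quality-level switch sends its first two slots to one modulator with a custom pitch minimum and the last two to another.
# Each modulator plays an equally weighted random pick of two like-named daytime takes.
# Bird one-shot attenuation applies, with a distance-scaled send to the exterior ambient source bus.
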SoundCue /Game/Sounds/Helios/Biomes/BorealTundra/Spots/Spot_Bird_Red_Throated_Loon_Singing_Cue
root=QualityLevel_0
nodes [
 {id=Modulator_3, children [Random_2], properties {pitchMin=0.9}}
 {id=Modulator_4, children [Random_3]}
 {id=QualityLevel_0, children [Modulator_3,Modulator_3,Modulator_4,Modulator_4]}
 {id=Random_2, children [WavePlayer_14, WavePlayer_15], properties {weights [1,1]}}
 {id=Random_3, children [WavePlayer_16, WavePlayer_17], properties {bShouldExcludeFromBranchCulling=true, weights [1,1]}}
 {id=WavePlayer_14, properties {soundWaveAssetPtr=&/Spot_Bird_Red_Throated_Loon_Singing_Day_02}}
 {id=WavePlayer_15, properties {soundWaveAssetPtr=&/Spot_Bird_Red_Throated_Loon_Singing_Day_01}}
 {id=WavePlayer_16, properties {soundWaveAssetPtr=&Spot_Bird_Red_Throated_Loon_Singing_Day_02}}
 {id=WavePlayer_17, properties {soundWaveAssetPtr=&Spot_Bird_Red_Throated_Loon_Singing_Day_01}}
]
properties {
 attenuationSettings=&Fort_TileAmbient_OneShot_Bird_Att
 maxDistance=8005
 soundClassObject=&Fort_Ambient_OneShots_Outdoors
 volumeMultiplier=0.05
 busSends [{
   audioBus=null
   maxSendDistance=6000
   minSendDistance=3000
   minSendLevel=0.6
   sendLevel=1
   soundSourceBus=&AmbientAudio_Exterior_SourceBus
  }]
}
