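# Sound cue for a Snowy Owl hoot, an outdoor one-shot ambient spot sound in the BorealTundra biome.
# A quality-level switch sends its first two and last two tiers to separate modulated random branches.
# Each branch picks evenly between two hoot waves, using same-named assets at differing paths.
# Bird attenuation shapes falloff, and a distance-based send feeds the exterior ambient source bus.
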SoundCue /Game/Sounds/Helios/Biomes/BorealTundra/Spots/Spot_Bird_Snowy_Owl_Hoots_Cue
root=QualityLevel_0
nodes [
 {id=Modulator_3, children [Random_2]}
 {id=Modulator_4, children [Random_3]}
 {id=QualityLevel_0, children [Modulator_3,Modulator_3,Modulator_4,Modulator_4]}
 {id=Random_2, children [WavePlayer_12, WavePlayer_13], properties {weights [1,1]}}
 {id=Random_3, children [WavePlayer_14, WavePlayer_15], properties {weights [1,1]}}
 {id=WavePlayer_12, properties {soundWaveAssetPtr=&/Spot_Bird_Snowy_Owl_Hoots_01}}
 {id=WavePlayer_13, properties {soundWaveAssetPtr=&/Spot_Bird_Snowy_Owl_Hoots_02}}
 {id=WavePlayer_14, properties {soundWaveAssetPtr=&Spot_Bird_Snowy_Owl_Hoots_01}}
 {id=WavePlayer_15, properties {soundWaveAssetPtr=&Spot_Bird_Snowy_Owl_Hoots_02}}
]
properties {
 attenuationSettings=&Fort_TileAmbient_OneShot_Bird_Att
 maxDistance=8005
 soundClassObject=&Fort_Ambient_OneShots_Outdoors
 volumeMultiplier=0.251
 busSends [{
   audioBus=null
   maxSendDistance=6000
   minSendDistance=3000
   minSendLevel=0.6
   sendLevel=1
   soundSourceBus=&AmbientAudio_Exterior_SourceBus
  }]
}
