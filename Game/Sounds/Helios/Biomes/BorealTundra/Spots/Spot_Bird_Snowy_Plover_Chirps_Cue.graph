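# Sound cue for snowy plover chirp one-shots, an outdoor ambient bird spot sound in the Helios Boreal Tundra biome.
# A quality node sends its first two levels and last two levels to separate Modulator-over-Random branches, each picking equally from six chirp variations.
# Mixers rebalance individual waves, with variation six attenuated only in the first branch.
# Playback uses bird one-shot attenuation and a distance-scaled send to the exterior ambience source bus.
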SoundCue /Game/Sounds/Helios/Biomes/BorealTundra/Spots/Spot_Bird_Snowy_Plover_Chirps_Cue
root=QualityLevel_0
nodes [
 {id=Mixer_0, children [WavePlayer_2], properties {inputVolume [1.6]}}
 {id=Mixer_1, children [WavePlayer_8], properties {inputVolume [1.6]}}
 {id=Mixer_2, children [WavePlayer_5], properties {inputVolume [0.6]}}
 {id=Modulator_3, children [Random_0]}
 {id=Modulator_4, children [Random_1]}
 {id=QualityLevel_0, children [Modulator_3,Modulator_3,Modulator_4,Modulator_4]}
 {id=Random_0, children [WavePlayer_0,WavePlayer_1,Mixer_0,WavePlayer_3,WavePlayer_4,Mixer_2], properties {weights [1,1,1,1,1,1]}}
 {
  id=Random_1
  children [WavePlayer_6, WavePlayer_7, Mixer_1, WavePlayer_9, WavePlayer_10, WavePlayer_11]
  properties {bShouldExcludeFromBranchCulling=true, weights [1,1,1,1,1,1]}
 }
 {id=WavePlayer_0, properties {soundWaveAssetPtr=&/Spot_Bird_Snowy_Plover_Chirps_01}}
 {id=WavePlayer_1, properties {soundWaveAssetPtr=&/Spot_Bird_Snowy_Plover_Chirps_02}}
 {id=WavePlayer_10, properties {soundWaveAssetPtr=&/Spot_Bird_Snowy_Plover_Chirps_05}}
 {id=WavePlayer_11, properties {soundWaveAssetPtr=&/Spot_Bird_Snowy_Plover_Chirps_06}}
 {id=WavePlayer_2, properties {soundWaveAssetPtr=&/Spot_Bird_Snowy_Plover_Chirps_03}}
 {id=WavePlayer_3, properties {soundWaveAssetPtr=&/Spot_Bird_Snowy_Plover_Chirps_04}}
 {id=WavePlayer_4, properties {soundWaveAssetPtr=&Spot_Bird_Snowy_Plover_Chirps_05}}
 {id=WavePlayer_5, properties {soundWaveAssetPtr=&Spot_Bird_Snowy_Plover_Chirps_06}}
 {id=WavePlayer_6, properties {soundWaveAssetPtr=&Spot_Bird_Snowy_Plover_Chirps_01}}
 {id=WavePlayer_7, properties {soundWaveAssetPtr=&Spot_Bird_Snowy_Plover_Chirps_02}}
 {id=WavePlayer_8, properties {soundWaveAssetPtr=&Spot_Bird_Snowy_Plover_Chirps_03}}
 {id=WavePlayer_9, properties {soundWaveAssetPtr=&Spot_Bird_Snowy_Plover_Chirps_04}}
]
properties {
 attenuationSettings=&Fort_TileAmbient_OneShot_Bird_Att
 maxDistance=8005
 soundClassObject=&Fort_Ambient_OneShots_Outdoors
 volumeMultiplier=0.2
 busSends [{
   audioBus=null
   maxSendDistance=6000
   minSendDistance=3000
   minSendLevel=0.6
   sendLevel=1
   soundSourceBus=&AmbientAudio_Exterior_SourceBus
  }]
}
