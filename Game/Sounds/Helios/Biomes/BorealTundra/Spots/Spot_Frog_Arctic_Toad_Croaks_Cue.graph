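# Sound cue for quiet arctic toad croak one-shots, a spot ambience in the Helios boreal tundra biome.
# A quality-level node sends the lower two and upper two levels to separate branches, each randomly picking one of four equally weighted, like-named croak waves from differing paths.
# A modulator varies each branch, with a lower pitch ceiling on the first.
# Playback uses outdoor ambient one-shot class and bird one-shot attenuation.
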SoundCue /Game/Sounds/Helios/Biomes/BorealTundra/Spots/Spot_Frog_Arctic_Toad_Croaks_Cue
root=QualityLevel_0
nodes [
 {id=Modulator_3, children [Random_2], properties {pitchMax=1}}
 {id=Modulator_4, children [Random_3]}
 {id=QualityLevel_0, children [Modulator_3,Modulator_3,Modulator_4,Modulator_4]}
 {id=Random_2, children [WavePlayer_12, WavePlayer_13, WavePlayer_14, WavePlayer_15], properties {weights [1,1,1,1]}}
 {id=Random_3, children [WavePlayer_16, WavePlayer_17, WavePlayer_18, WavePlayer_19], properties {weights [1,1,1,1]}}
 {id=WavePlayer_12, properties {soundWaveAssetPtr=&/Spot_Frog_Arctic_Toad_Croaks_01}}
 {id=WavePlayer_13, properties {soundWaveAssetPtr=&/Spot_Frog_Arctic_Toad_Croaks_02}}
 {id=WavePlayer_14, properties {soundWaveAssetPtr=&/Spot_Frog_Arctic_Toad_Croaks_03}}
 {id=WavePlayer_15, properties {soundWaveAssetPtr=&/Spot_Frog_Arctic_Toad_Croaks_04}}
 {id=WavePlayer_16, properties {soundWaveAssetPtr=&Spot_Frog_Arctic_Toad_Croaks_01}}
 {id=WavePlayer_17, properties {soundWaveAssetPtr=&Spot_Frog_Arctic_Toad_Croaks_02}}
 {id=WavePlayer_18, properties {soundWaveAssetPtr=&Spot_Frog_Arctic_Toad_Croaks_03}}
 {id=WavePlayer_19, properties {soundWaveAssetPtr=&Spot_Frog_Arctic_Toad_Croaks_04}}
]
properties {
 attenuationSettings=&Fort_TileAmbient_OneShot_Bird_Att
 bEnableBusSends=false
 maxDistance=8005
 soundClassObject=&Fort_Ambient_OneShots_Outdoors
 volumeMultiplier=0.1
 preEffectBusSends [{}]
}
